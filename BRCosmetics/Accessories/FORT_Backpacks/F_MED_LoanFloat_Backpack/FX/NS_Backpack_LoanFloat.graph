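# Niagara effect for the LoanFloat backpack cosmetic, with an attached ribbon and steam sprites that follow the system's skeletal mesh.
# The ribbon drifts upward and offsets against owner velocity, less while falling, with opacity driven by timeline, fade-in, camera angle and time of day.
# Steam spawns probabilistically only while the pawn-landed flag is set, growing and fading over life and with distance from the owner.
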NiagaraSystem /BRCosmetics/Accessories/FORT_Backpacks/F_MED_LoanFloat_Backpack/FX/NS_Backpack_LoanFloat
dependencies {
 usedDataInterfaces [
  &NiagaraDataInterfaceCurve
  &NiagaraDataInterfaceColorCurve
  &NiagaraDataInterfaceCamera
  &NiagaraDataInterfaceVectorCurve
  &NiagaraDataInterfaceSkeletalMesh
  &NiagaraDataInterfaceArrayFloat
 ]
 usedDynamicInputs [
  &Subtract_Float
  &SetBoolByFloatComparison
  &VectorLength
 ]
 usedModules [
  &self:SystemSpawnScript.SystemScriptSource.SystemScriptGraph.NiagaraNodeAssignment_0.SetVariables_#1
  &self:SystemSpawnScript.SystemScriptSource.SystemScriptGraph.NiagaraNodeAssignment_1.SetVariables_#2
  &SystemState
  &self:SystemSpawnScript.SystemScriptSource.SystemScriptGraph.NiagaraNodeAssignment_2.SetVariables_#3
  &Timeline
 ]
}
emitters {
 FN_AttachedRibbon {
  emitterUpdateScript [
   &EmitterState()
   &FN_SpawnRibbon_v2(SpawnRate=30)
  ]
  particleSpawnScript [
   /Game/Effects/Niagara/Modules/Ribbons/FN_InitializeRibbon_v4(
    Custom Offset=true
    Enable Skeletal Position=true
    Lifetime=0.6
    Local Vector Offset:Vector3f=<0 -1 1>
    Ribbon Width=15
    SkeletalMesh:NiagaraDataInterfaceSkeletalMesh=$System.Skeletal Mesh
   )
   &FN_FlipRibbonV()
   &AddVelocity(Coordinate Space=::Simulation, Velocity=<35 -35 35>) {enabled=false}
   &CameraOffset(Camera Offset Amount=31.7356)
  ]
  particleUpdateScript [
   &InheritSourceMovement(
    Applied Position Scale=&Multiply_VectorByFloat(
     Vector=<-1 -1 -0.3>
     Float=&FloatFromCurve(
      FloatCurve=DataInterface() {propertyValues {Curve {
         defaultValue=0
         keys [
          {arriveTangent=-0.0044, interpMode=RCIM_Cubic, leaveTangent=-0.0044, tangentMode=RCTM_User, time=0, value=0}
          {arriveTangent=0.0645, interpMode=RCIM_Cubic, leaveTangent=0.0645, tangentMode=RCTM_User, time=0.2462, value=0}
          {arriveTangent=1.437, interpMode=RCIM_Cubic, leaveTangent=1.437, tangentMode=RCTM_User, time=1.002, value=0.9882}
         ]
        }}}
      Scale Curve=&MakeCustomFloatFromBool(Bool=$User.bPawnFalling, True Float=0.2, False Float=0.5)
     )
    )
    ApplyForce=false
    ApplyPosition=true
    ApplyPositionFalloff=false
    Source Velocity:Vector3f=&TransformVector(Vector=$Engine.Owner.Velocity, Source Space=::World, Destination Space=::Local)
   )
   &GravityForce(Gravity=<0 0 250>)
   &Drag(Drag=2, Use Linear Drag=true, Use Rotational Drag=true)
   &SolveForcesAndVelocity(Clamp Velocity=false, Limit Acceleration=false)
   /Game/Effects/Niagara/Modules/Ribbons/FN_UpdateRibbon_v4(Enable Skeletal Position=true, SkeletalMesh:NiagaraDataInterfaceSkeletalMesh=$System.Skeletal Mesh)
   &FN_RibbonMaterial(
    Color=&ColorFromCurve(ColorCurve=DataInterface() {propertyValues {
       AlphaCurve {keys [<at 0 1 RCIM_Cubic>]}
       BlueCurve {keys [<at 0.4279 0.3542 RCIM_Cubic>]}
       GreenCurve {keys [<at 0.4279 0.3542 RCIM_Cubic>]}
       RedCurve {keys [<at 0.4279 0.3542 RCIM_Cubic>]}
      }})
    Color Pan U Speed=0.1
    Opacity Pan U Speed=0.5
    Opacity Power=0.7
    Opacity Scale U=0.7
    Scale Alpha:NiagaraFloat=&FloatFromCurve(
     FloatCurve=DataInterface() {propertyValues {Curve {keys [{interpMode=RCIM_Cubic, tangentMode=RCTM_User, time=0.3, value=1}, {interpMode=RCIM_Cubic, tangentMode=RCTM_User, time=1, value=0}]}}}
     Scale Curve=&Multiply_Float(
      A=&Multiply_Float(
       A=$System.Timeline.Float
       B=&MakeFloatFromVector4(
        Vector 4=&MakeVector4_FromFloat(Float=&Multiply_Float(
          A=&Multiply_Float(A=1.5, B=$System.FadeIn)
          B=&ClampFloat(
           Float=&Multiply_Float(
            A=&DotProduct(Vector A=&ReturnCameraVector(Camera Vectors=::NewEnumerator1, Coordinate Space=::Local), Vector B=<0 1 0>)
            B=1
           )
           Min=0.6
          )
         ))
        Channel=::NewEnumerator0
       )
      )
      B=&FNTimeOfDayScalar(Night Scale=1.2, Morning Scale=1, Day Scale=0.7, Evening Scale=1)
     )
    )
    Scale Color:Vector3f=&MakeVectorFromLinearColorRGB(Color=$System.TODColors)
   ) {module=FN_RibbonMaterial002}
  ]
  rendererClasses [&NiagaraRibbonRendererProperties]
  renderers [{rendererClass=&NiagaraRibbonRendererProperties}]
 }
 Steam {
  emitterUpdateScript [
   &EmitterState(Life Cycle Mode=::NewEnumerator1, Loop Duration=1, UseLoopDelay=true)
   &SpawnPerFrame(
    Spawn Probability:NiagaraFloat=&MakeCustomFloatFromBool(Bool=$User.bPawnLanded, True Float=0.75, False Float=0)
    Use Spawn Probability=true
   )
  ]
  particleSpawnScript [
   &InitializeParticle(
    Color=<rgba 1 0.9305 0.8353 0.05>
    Lifetime Max=1.25
    Lifetime Min=0.75
    Lifetime Mode=::NewEnumerator1
    Sprite Rotation Mode=::NewEnumerator2
    Sprite Size Mode=::NewEnumerator1
    Uniform Sprite Size Max=35
    Uniform Sprite Size Min=30
    UsePositionOffset=false
    Write Material Random=true
   )
   /SetVariables(Particles.InitRandomFloat:NiagaraFloat=&RandomRangeFloat(Minimum=0, Maximum=1)) {isSetParameters=true}
   &AddVelocity(Rotation Coordinate Space=::World)
   &SkeletalMeshLocation(MaskSpawnGroup=false, Mesh Sampling Type=::NewEnumerator5, Skeletal Mesh=$System.Skeletal Mesh, Socket Sampling Mode=::NewEnumerator5)
   &CameraOffset(Camera Offset Amount=50) {enabled=false}
  ]
  particleUpdateScript [
   &ParticleState()
   &DynamicMaterialParameters(
    Index 0 Param 1=&FloatFromCurve(
     FloatCurve=DataInterface() {propertyValues {Curve {keys [<at 0 0>,<at 1 1>]}}}
     Scale Curve=&RandomRangeFloat(Minimum=0.1, Maximum=0.3)
    )
    Index 0 Param 4=&FloatFromCurve(FloatCurve=DataInterface() {propertyValues {Curve {keys [<at 0 1>,<at 1 0>]}}}, Scale Curve=0.25)
    Index 2 Param 1=&FloatFromCurve(FloatCurve=DataInterface() {propertyValues {Curve {keys [
         {arriveTangent=2.8586, interpMode=RCIM_Cubic, leaveTangent=2.8586, tangentMode=RCTM_User, time=0, value=0}
         {interpMode=RCIM_Cubic, tangentMode=RCTM_User, time=1, value=1}
        ]}}})
    Index 2 Param 2=&FloatFromCurve(FloatCurve=DataInterface() {propertyValues {Curve {keys [<at 0 1>,<at 1 0>]}}}, Scale Curve=0.2)
    Index 2 Param 3=1
    Index 2 Param 4=0.5
    Module.Param0WriteEnabled=true
    Module.Param0WriteEnabled2=true
    Module.Param1WriteEnabled=false
    Module.Param1WriteEnabled2=true
    Module.Param2WriteEnabled=false
    Module.Param2WriteEnabled2=true
    Module.Param3WriteEnabled=true
    Module.Param3WriteEnabled2=true
    Write Parameter Index 2=true
   )
   &ScaleSpriteSize(Scale Factor=&Vector2DFromFloat(Value=&FloatFromCurve(FloatCurve=DataInterface() {propertyValues {Curve {keys [
          {arriveTangent=1.4006, interpMode=RCIM_Cubic, leaveTangent=1.4006, tangentMode=RCTM_User, time=0, value=0.7}
          {arriveTangent=0.3317, interpMode=RCIM_Cubic, leaveTangent=0.3317, tangentMode=RCTM_User, time=1, value=1.2}
         ]}}})))
   &Drag(Drag=2, Use Linear Drag=true, Use Rotational Drag=false)
   &InheritSourceMovement(
    Applied Position Scale=&Multiply_Vector(
     A=<1 1 1>
     B=&VectorFromCurve(VectorCurve=DataInterface() {propertyValues {XCurve {keys [<at 0 1>,<at 1 0.6>]}, YCurve {keys [<at 0 1>,<at 1 0.6>]}, ZCurve {keys [<at 0 0.2>,<at 1 0.2>]}}})
    )
    ApplyForce=false
    ApplyPosition=true
    ApplyPositionFalloff=false
   )
   &SolveForcesAndVelocity(Clamp Velocity=false, Limit Acceleration=false)
   &ScaleColor(
    Scale Alpha=&Multiply_Float(
     A=&NormalizeDistanceRange(
      Start Position:NiagaraPosition=$Particles.Position
      End Position:NiagaraPosition=$Engine.Owner.Position
      Distance:NiagaraFloat=50
      Invert Normalized Range:NiagaraBool=true
     )
     B=&FloatFromCurve(
      FloatCurve=DataInterface() {propertyValues {Curve {keys [
          {arriveTangent=4.231, interpMode=RCIM_Cubic, leaveTangent=4.231, tangentMode=RCTM_User, time=0, value=0}
          {arriveTangent=0.1173, interpMode=RCIM_Cubic, leaveTangent=0.1173, tangentMode=RCTM_User, time=0.3, value=1}
          {arriveTangent=-2.9749, interpMode=RCIM_Cubic, leaveTangent=-2.9749, tangentMode=RCTM_User, time=1, value=0}
         ]}}}
      Scale Curve=0.5
     )
    )
    ScaleA=true
    ScaleRGB=true
   )
   &self:GoBackToPos_01(
    PositionGoal:NiagaraPosition=&AddVectorToPosition(
     Position=$Particles.SkeletalMeshLocation.SampledPosition
     Vector=&RandomRangeVector(Minimum=<0 0 15>, Maximum=<0 0 25>)
    )
    ZrestSpeed:NiagaraFloat=&FloatFromCurve(FloatCurve=DataInterface() {propertyValues {Curve {keys [<at 0 0.6>,<at 1 0.2>]}}})
   ) {enabled=false, module=GoBackToPos_01}
  ]
  rendererClasses [&NiagaraSpriteRendererProperties]
  renderers [
   {rendererClass=&NiagaraSpriteRendererProperties}
   {rendererClass=&NiagaraSpriteRendererProperties, rendererIndex=1}
  ]
 }
}
summary {
 systemName=NS_Backpack_LoanFloat
 emitters [
  {emitterName=FN_AttachedRibbon, rendererClasses [&NiagaraRibbonRendererProperties]}
  {emitterName=Steam, rendererClasses [&NiagaraSpriteRendererProperties]}
 ]
 userVariables [
  User.bIsFrontEnd {type=NiagaraBool}
  User.bIsFrontEndPreview {type=NiagaraBool}
  User.bPawnFalling {type=NiagaraBool}
  User.bPawnLanded {type=NiagaraBool}
  User.RibbonIDOffset {type=NiagaraInt32}
 ]
}
